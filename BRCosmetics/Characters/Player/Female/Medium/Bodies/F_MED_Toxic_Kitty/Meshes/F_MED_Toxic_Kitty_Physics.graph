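# Physics asset for the Toxic Kitty female medium body cosmetic, covering its dangling rear pack, tubes and links.
# Simple capsule, box and sphere bodies cover only pelvis, spine_04 and the dyn_ accessory bones.
# Rear pack and right tube chain to pelvis, low tube to spine_04, via small parent bodies with free rotation.
# Accessory end joints, including both links hung directly on spine_04, have limited swing and tighter twist.
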PhysicsAsset /BRCosmetics/Characters/Player/Female/Medium/Bodies/F_MED_Toxic_Kitty/Meshes/F_MED_Toxic_Kitty_Physics
bodies {
 dyn_linkA [{shapeType=Capsule, length=5.25, radius=3.75, center=<0 -0.9514 -0.1605>, rotation=<rot 0 0 -18.0347>}]
 dyn_linkB [{shapeType=Capsule, length=8.25, radius=3.25, center=<0.0368 -0.5582 -0.2264>, rotation=<rot -2.251 0.8449 -8.3636>}]
 dyn_pack_rear_01 [{shapeType=Box, extentX=13.5, extentY=1.5, extentZ=20.5, center=<3.9967 -0.5512 0.2941>, rotation=<rot -1.9532 0 0>}]
 dyn_parent_pack_rear [{shapeType=Sphere, radius=0.505}]
 dyn_parent_tube_low [{shapeType=Sphere, radius=0.505}]
 dyn_parent_tube_r [{shapeType=Box, extentX=1.01, extentY=1.01, extentZ=1.01, rotation=<rot 0 -90 0>}]
 dyn_tube_01_r [{shapeType=Box, extentX=2.01, extentY=23.01, extentZ=18.51, center=<10.1602 -2.6963 -0.0002>, rotation=<rot 0 -103.0673 0>}]
 dyn_tube_low_01 [{shapeType=Capsule, length=6.5, radius=8, center=<7 0 0.0001>, rotation=<rot 90 0 21.4553>}]
 pelvis [{shapeType=Capsule, length=8.25, radius=10.5, center=<-2.5008 -1.899 0.7099>}]
 spine_04 [{shapeType=Sphere, radius=1.0692, center=<-1.4985 2.8226 -1.1009>}]
}
constraints [
 {
  bone1Name=dyn_parent_pack_rear
  bone2Name=pelvis
  swing1LimitDegrees=45
  swing1Motion=Free
  swing2LimitDegrees=45
  swing2Motion=Free
  twistLimitDegrees=45
  twistMotion=Free
 }
 {bone1Name=dyn_pack_rear_01, bone2Name=dyn_parent_pack_rear, swing1LimitDegrees=65.5714, swing2LimitDegrees=41.5714, twistLimitDegrees=15}
 {
  bone1Name=dyn_parent_tube_r
  bone2Name=pelvis
  swing1LimitDegrees=45
  swing1Motion=Free
  swing2LimitDegrees=45
  swing2Motion=Free
  twistLimitDegrees=45
  twistMotion=Free
 }
 {bone1Name=dyn_tube_01_r, bone2Name=dyn_parent_tube_r, swing1LimitDegrees=45, swing2LimitDegrees=45, twistLimitDegrees=20}
 {
  bone1Name=spine_04
  bone2Name=pelvis
  swing1LimitDegrees=45
  swing1Motion=Free
  swing2LimitDegrees=45
  swing2Motion=Free
  twistLimitDegrees=45
  twistMotion=Free
 }
 {bone1Name=dyn_linkA, bone2Name=spine_04, swing1LimitDegrees=45, swing2LimitDegrees=45, twistLimitDegrees=15}
 {bone1Name=dyn_linkB, bone2Name=spine_04, swing1LimitDegrees=45, swing2LimitDegrees=45, twistLimitDegrees=15}
 {
  bone1Name=dyn_parent_tube_low
  bone2Name=spine_04
  swing1LimitDegrees=45
  swing1Motion=Free
  swing2LimitDegrees=45
  swing2Motion=Free
  twistLimitDegrees=45
  twistMotion=Free
 }
 {bone1Name=dyn_tube_low_01, bone2Name=dyn_parent_tube_low, swing1LimitDegrees=45, swing2LimitDegrees=45, twistLimitDegrees=11}
]
