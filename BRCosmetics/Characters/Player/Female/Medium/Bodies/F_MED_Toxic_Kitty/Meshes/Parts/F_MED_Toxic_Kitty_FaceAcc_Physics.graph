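# Physics asset for the Toxic Kitty female medium character's FaceAcc mesh part, providing bodies and joints for dynamic back hair.
# Capsules on the head, upper spine and clavicles form body shapes around three chains of thin boxes for center, left and right hair.
# Chain joints carry angular limits, while the side roots at the head swing and twist freely.
# Angularly free cross-constraints tie the center chain's third bone to each side chain's third bone.
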PhysicsAsset /BRCosmetics/Characters/Player/Female/Medium/Bodies/F_MED_Toxic_Kitty/Meshes/Parts/F_MED_Toxic_Kitty_FaceAcc_Physics
bodies {
 clavicle_l [{shapeType=Capsule, length=2.005, radius=6.755, center=<4.187 -0.9653 0.2573>, rotation=<rot 45.9377 -89.9995 0>}]
 clavicle_r [{shapeType=Capsule, length=1.255, radius=7.505, center=<-3.2182 0.8122 -0.0993>, rotation=<rot 45.9373 -89.9989 180>}]
 dyn_hair_bk_01 [{shapeType=Box, extentX=9, extentY=1, extentZ=9.5}]
 dyn_hair_bk_02 [{shapeType=Box, extentX=11.5, extentY=1.05, extentZ=14.5, center=<4.837 0.8531 0>, rotation=<rot 0 10.0012 0>}]
 dyn_hair_bk_03 [{shapeType=Box, extentX=9, extentY=1.05, extentZ=13.5, center=<4.4869 0.0001 0>}]
 dyn_hair_bk_04 [{shapeType=Box, extentX=7.5, extentY=1.05, extentZ=10, center=<4.5451 0 0>}]
 dyn_hair_bk_05 [{shapeType=Box, extentX=7, extentY=1.05, extentZ=8.5, center=<4.8388 -0.0002 0>}]
 dyn_hair_bk_l_01 [{shapeType=Box, extentX=7.5, extentY=1.1, extentZ=12.5}]
 dyn_hair_bk_l_02 [{shapeType=Box, extentX=9.5, extentY=1.1, extentZ=13, center=<1.0364 -0.0001 0>}]
 dyn_hair_bk_l_03 [{shapeType=Box, extentX=9.5, extentY=1.1, extentZ=10, center=<1.2085 -0.0004 -0.0001>, rotation=<rot 0 0 -13.3259>}]
 dyn_hair_bk_l_04 [{shapeType=Box, extentX=9.5, extentY=1.1, extentZ=10, center=<2.0381 -0.0004 -0.0001>, rotation=<rot 0 0 -10.2184>}]
 dyn_hair_bk_r_01 [{shapeType=Box, extentX=8.5, extentY=1.1, extentZ=10}]
 dyn_hair_bk_r_02 [{shapeType=Box, extentX=8.5, extentY=1.1, extentZ=10, center=<1.5957 0 0>, rotation=<rot 0 4.9693 0>}]
 dyn_hair_bk_r_03 [{shapeType=Box, extentX=8.5, extentY=2.1, extentZ=14.5, center=<1.9284 -0.3699 -0.0001>}]
 head [{shapeType=Capsule, length=4.75, radius=7, center=<-0.3946 2.2377 0>, rotation=<rot -90 90 -79.9997>}]
 spine_05 [{shapeType=Capsule, length=4.255, radius=16.755, center=<-0.6865 5.7414 0>, rotation=<rot 0 -90 0>}]
}
constraints [
 {bone1Name=dyn_hair_bk_01, bone2Name=head, swing1LimitDegrees=45, swing2LimitDegrees=45, twistLimitDegrees=45}
 {bone1Name=dyn_hair_bk_02, bone2Name=dyn_hair_bk_01, swing1LimitDegrees=45, swing2LimitDegrees=45, twistLimitDegrees=45}
 {bone1Name=dyn_hair_bk_03, bone2Name=dyn_hair_bk_02, swing1LimitDegrees=45, swing2LimitDegrees=45, twistLimitDegrees=45}
 {bone1Name=dyn_hair_bk_04, bone2Name=dyn_hair_bk_03, swing1LimitDegrees=45, swing2LimitDegrees=45, twistLimitDegrees=45}
 {bone1Name=dyn_hair_bk_05, bone2Name=dyn_hair_bk_04, swing1LimitDegrees=45, swing2LimitDegrees=45, twistLimitDegrees=45}
 {
  bone1Name=head
  bone2Name=spine_05
  swing1LimitDegrees=45
  swing1Motion=Free
  swing2LimitDegrees=45
  swing2Motion=Free
  twistLimitDegrees=45
  twistMotion=Free
 }
 {
  bone1Name=dyn_hair_bk_l_01
  bone2Name=head
  swing1LimitDegrees=45
  swing1Motion=Free
  swing2LimitDegrees=45
  swing2Motion=Free
  twistLimitDegrees=45
  twistMotion=Free
 }
 {bone1Name=dyn_hair_bk_l_02, bone2Name=dyn_hair_bk_l_01, swing1LimitDegrees=45, swing2LimitDegrees=45, twistLimitDegrees=45}
 {bone1Name=dyn_hair_bk_l_03, bone2Name=dyn_hair_bk_l_02, swing1LimitDegrees=45, swing2LimitDegrees=45, twistLimitDegrees=45}
 {bone1Name=dyn_hair_bk_l_04, bone2Name=dyn_hair_bk_l_03, swing1LimitDegrees=45, swing2LimitDegrees=45, twistLimitDegrees=45}
 {
  bone1Name=dyn_hair_bk_03
  bone2Name=dyn_hair_bk_l_03
  swing1LimitDegrees=45
  swing1Motion=Free
  swing2LimitDegrees=45
  swing2Motion=Free
  twistLimitDegrees=45
  twistMotion=Free
 }
 {
  bone1Name=clavicle_l
  bone2Name=spine_05
  swing1LimitDegrees=45
  swing1Motion=Free
  swing2LimitDegrees=45
  swing2Motion=Free
  twistLimitDegrees=45
  twistMotion=Free
 }
 {
  bone1Name=clavicle_r
  bone2Name=spine_05
  swing1LimitDegrees=45
  swing1Motion=Free
  swing2LimitDegrees=45
  swing2Motion=Free
  twistLimitDegrees=45
  twistMotion=Free
 }
 {
  bone1Name=dyn_hair_bk_r_01
  bone2Name=head
  swing1LimitDegrees=45
  swing1Motion=Free
  swing2LimitDegrees=45
  swing2Motion=Free
  twistLimitDegrees=45
  twistMotion=Free
 }
 {bone1Name=dyn_hair_bk_r_03, bone2Name=dyn_hair_bk_r_02, swing1LimitDegrees=45, swing2LimitDegrees=45, twistLimitDegrees=45}
 {bone1Name=dyn_hair_bk_r_02, bone2Name=dyn_hair_bk_r_01, swing1LimitDegrees=45, swing2LimitDegrees=45, twistLimitDegrees=45}
 {
  bone1Name=dyn_hair_bk_03
  bone2Name=dyn_hair_bk_r_03
  swing1LimitDegrees=45
  swing1Motion=Free
  swing2LimitDegrees=45
  swing2Motion=Free
  twistLimitDegrees=45
  twistMotion=Free
 }
]
