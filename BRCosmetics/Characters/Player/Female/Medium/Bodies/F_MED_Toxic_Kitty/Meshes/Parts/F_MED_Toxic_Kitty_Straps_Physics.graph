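# Physics asset for the straps on the Toxic Kitty female medium character body, defining their collision and joint limits.
# Two five-capsule chains, strapA and strapB, branch from a small driver body attached to the head, with tight twist and swing2 limits that widen toward the tips.
# Head, neck, upper spine and two back-hair bodies add capsule and box volumes, linked by free angular constraints, as is the driver.
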PhysicsAsset /BRCosmetics/Characters/Player/Female/Medium/Bodies/F_MED_Toxic_Kitty/Meshes/Parts/F_MED_Toxic_Kitty_Straps_Physics
bodies {
 dyn_hair_bk_03 [{shapeType=Box, extentX=10, extentY=1.5, extentZ=16.5, center=<-4.7837 0.598 -0.0002>}]
 dyn_hair_bk_04 [
  {shapeType=Capsule, length=6.5, radius=19.75, center=<1.1596 19.1875 0>, rotation=<rot 0 -90 -90>}
  {shapeType=Box, extentX=18.5, extentY=1.5, extentZ=15.5, center=<0.0197 1.9547 -0.24>, rotation=<rot 0 0.0004 0>}
 ]
 dyn_strapA_01 [{shapeType=Capsule, length=5.25, radius=1.505, center=<0.0999 -0.0001 1.4407>, rotation=<rot 0.1489 92.1577 -93.9641>}]
 dyn_strapA_02 [{shapeType=Capsule, length=8.662, radius=1.7037, center=<3.1554 -0.0123 1.1872>, rotation=<rot 0 -90.0528 -88.5289>}]
 dyn_strapA_03 [{shapeType=Capsule, length=6.3219, radius=1.6719, center=<4.0052 -0.1224 1.1961>, rotation=<rot -0.0275 88.2153 -90.1595>}]
 dyn_strapA_04 [{shapeType=Capsule, length=4.5616, radius=2.5604, center=<1.7996 -0.1196 1.6764>, rotation=<rot -1.1406 87.4169 -89.3164>}]
 dyn_strapA_05 [{shapeType=Capsule, length=3.7861, radius=2.5673, center=<1.9744 -0.0009 1.1879>, rotation=<rot -7.5493 87.9061 -90.9785>}]
 dyn_strapB_01 [{shapeType=Capsule, length=3.25, radius=1.255, center=<2.2274 -0.2873 0.2652>, rotation=<rot 20.8603 89.0722 -89.9999>}]
 dyn_strapB_02 [{shapeType=Capsule, length=4.1841, radius=1.2615, center=<1.43 -0.0868 0.6019>, rotation=<rot 0 87.5877 -91.0878>}]
 dyn_strapB_03 [{shapeType=Capsule, length=4.9415, radius=1.5867, center=<2.3054 -0.0879 0.7406>, rotation=<rot 0 -89.935 -89.0539>}]
 dyn_strapB_04 [{shapeType=Capsule, length=5.094, radius=1.7971, center=<2.7849 -0.0203 0.8031>, rotation=<rot 0 89.4277 -89.6783>}]
 dyn_strapB_05 [{shapeType=Capsule, length=7.6212, radius=2.3654, center=<1.0994 0.5237 1.5686>, rotation=<rot 16.0728 -84.7393 -90.6233>}]
 dyn_strap_driver [{shapeType=Capsule, length=0.505, radius=0.505, rotation=<rot 0 -90 0>}]
 head [{shapeType=Capsule, length=3, radius=7.91}]
 neck_01 [{shapeType=Capsule, length=6.5, radius=10.75, center=<-2.1828 3.1433 -0.0001>}]
 spine_05 [{shapeType=Capsule, length=1, radius=23.005, center=<-0.4621 13.1479 -0.0001>, rotation=<rot 0 -90 0>}]
}
constraints [
 {bone1Name=dyn_strapB_01, bone2Name=dyn_strap_driver, swing1LimitDegrees=45, swing2LimitDegrees=35, twistLimitDegrees=5}
 {bone1Name=dyn_strapB_02, bone2Name=dyn_strapB_01, swing1LimitDegrees=45, swing2LimitDegrees=35, twistLimitDegrees=5}
 {bone1Name=dyn_strapB_03, bone2Name=dyn_strapB_02, swing1LimitDegrees=35, swing2LimitDegrees=69.2857, twistLimitDegrees=5}
 {bone1Name=dyn_strapB_04, bone2Name=dyn_strapB_03, swing1LimitDegrees=29.8571, swing2LimitDegrees=86.4286, twistLimitDegrees=5}
 {bone1Name=dyn_strapB_05, bone2Name=dyn_strapB_04, swing1LimitDegrees=45, swing2LimitDegrees=89.8571, twistLimitDegrees=5}
 {bone1Name=dyn_strapA_01, bone2Name=dyn_strap_driver, swing1LimitDegrees=35, swing2LimitDegrees=35, twistLimitDegrees=5}
 {bone1Name=dyn_strapA_02, bone2Name=dyn_strapA_01, swing1LimitDegrees=45, swing2LimitDegrees=35, twistLimitDegrees=5}
 {bone1Name=dyn_strapA_03, bone2Name=dyn_strapA_02, swing1LimitDegrees=35, swing2LimitDegrees=71, twistLimitDegrees=5}
 {bone1Name=dyn_strapA_04, bone2Name=dyn_strapA_03, swing1LimitDegrees=35, swing2LimitDegrees=88, twistLimitDegrees=5}
 {bone1Name=dyn_strapA_05, bone2Name=dyn_strapA_04, swing1LimitDegrees=45, swing2LimitDegrees=88.1428, twistLimitDegrees=5}
 {
  bone1Name=dyn_strap_driver
  bone2Name=head
  swing1LimitDegrees=45
  swing1Motion=Free
  swing2LimitDegrees=45
  swing2Motion=Free
  twistLimitDegrees=45
  twistMotion=Free
 }
 {
  bone1Name=dyn_hair_bk_04
  bone2Name=dyn_hair_bk_03
  swing1LimitDegrees=45
  swing1Motion=Free
  swing2LimitDegrees=45
  swing2Motion=Free
  twistLimitDegrees=45
  twistMotion=Free
 }
 {
  bone1Name=dyn_hair_bk_03
  bone2Name=head
  swing1LimitDegrees=45
  swing1Motion=Free
  swing2LimitDegrees=45
  swing2Motion=Free
  twistLimitDegrees=45
  twistMotion=Free
 }
 {
  bone1Name=head
  bone2Name=neck_01
  swing1LimitDegrees=45
  swing1Motion=Free
  swing2LimitDegrees=45
  swing2Motion=Free
  twistLimitDegrees=45
  twistMotion=Free
 }
 {
  bone1Name=neck_01
  bone2Name=spine_05
  swing1LimitDegrees=45
  swing1Motion=Free
  swing2LimitDegrees=45
  swing2Motion=Free
  twistLimitDegrees=45
  twistMotion=Free
 }
]
