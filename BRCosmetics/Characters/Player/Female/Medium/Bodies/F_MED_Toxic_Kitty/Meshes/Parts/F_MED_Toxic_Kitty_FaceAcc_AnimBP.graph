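# Post-processes the Toxic Kitty face accessory skeleton, adding secondary motion to hair and straps.
# Copies the attached parent's pose, then layers head-rotation-driven Flap corrections and spring-based Anim Dynamics on hair bones.
# From cached 'AnimCache', rigids_off selects between two strap branches: RigidBody simulation with a straps physics asset, or a fallback using Look At and plane-limited Trail controllers plus hair Flaps.
# Variable-driven bone offsets adjust the strap parent in both branches.
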AnimBlueprint /BRCosmetics/Characters/Player/Female/Medium/Bodies/F_MED_Toxic_Kitty/Meshes/Parts/F_MED_Toxic_Kitty_FaceAcc_AnimBP
target_skeleton=&/F_MED_Toxic_Kitty_FaceAcc_Skeleton
graphs [
 AnimGraph {nodes [
   n1 = Get hair_alpha() -> (hair_alpha:double)
   Anim Dynamics - Bone: dyn_hair_bk_02(Alpha=@n1, ExternalForce="0,0,0") {
    angularDampingOverride=0.9999
    angularSpringConstant=1000
    bAngularSpring=true
    bGravityOverrideInSimSpace=true
    bOverrideAngularDamping=true
    bUseSphericalLimits=true
    gravityScale=0.35
    simulationSpace=RootRelative
    alphaBoolBlend {blendOption=Linear}
    boundBone {boneName=dyn_hair_bk_02}
    componentAppliedLinearAccClamp=<100 100 100>
    componentLinearVelScale=<50 50 1>
    physicsBodyDefinitions [{
      collisionType=CustomSphere
      sphereCollisionRadius=3
      boundBone {boneName=dyn_hair_bk_02}
      boxExtents=<8 40 40>
      constraintSetup {twistAxis=AxisY, angularLimitsMax=<0 8 45>, angularLimitsMin=<0 -8 -75>, angularTarget=<0 0 0>}
      localJointOffset=<6 0 0>
     }]
    sphericalLimits [{limitRadius=40, drivingBone {boneName=spine_05}, sphereLocalOffset=<-12 26 0>}]
   }
   Anim Dynamics - Bone: dyn_hair_bk_01(Alpha=@n1, ExternalForce="0,0,0") {
    angularDampingOverride=0.9999
    angularSpringConstant=1000
    bAngularSpring=true
    bGravityOverrideInSimSpace=true
    bOverrideAngularDamping=true
    gravityScale=0.33
    alphaBoolBlend {blendOption=Linear}
    boundBone {boneName=dyn_hair_bk_01}
    chainEnd {boneName=dyn_hair_bk_02}
    componentAppliedLinearAccClamp=<100 100 100>
    componentLinearVelScale=<50 50 1>
    physicsBodyDefinitions [{
      collisionType=CustomSphere
      sphereCollisionRadius=4
      boundBone {boneName=dyn_hair_bk_01}
      boxExtents=<5 30 30>
      constraintSetup {angularLimitsMax=<0 25 3>, angularLimitsMin=<0 -25 -65>, angularTarget=<0 0 0>}
      localJointOffset=<5 0 0>
     }]
    sphericalLimits [
     {limitRadius=6.25, drivingBone {boneName=neck_02}, sphereLocalOffset=<0 0 0>}
     {limitRadius=6.25, drivingBone {boneName=neck_01}, sphereLocalOffset=<0 0 0>}
     {limitRadius=6.25, drivingBone {boneName=head}, sphereLocalOffset=<0 0 0>}
    ]
   }
   Anim Dynamics - Bone: dyn_hair_bk_05(Alpha=@n1, ExternalForce="0,0,0") {
    angularDampingOverride=0.9999
    angularSpringConstant=1000
    bAngularSpring=true
    bGravityOverrideInSimSpace=true
    bOverrideAngularDamping=true
    bUseSphericalLimits=true
    gravityScale=0.32
    simulationSpace=RootRelative
    alphaBoolBlend {blendOption=Linear}
    boundBone {boneName=dyn_hair_bk_05}
    componentAppliedLinearAccClamp=<100 100 100>
    componentLinearVelScale=<50 50 1>
    physicsBodyDefinitions [{
      collisionType=CustomSphere
      sphereCollisionRadius=2.5
      boundBone {boneName=dyn_hair_bk_05}
      boxExtents=<8 15 15>
      constraintSetup {angularLimitsMax=<0 20 35>, angularLimitsMin=<0 -20 -75>, angularTarget=<0 0 0>}
      localJointOffset=<2 0 0>
     }]
    sphericalLimits [{limitRadius=38, drivingBone {boneName=spine_05}, sphereLocalOffset=<-12 26 0>}]
   }
   n2 = Anim Dynamics - Bone: dyn_hair_bk_03(Alpha=1, ExternalForce="0,0,0") {
    angularDampingOverride=0.9999
    angularSpringConstant=1000
    bAngularSpring=true
    bGravityOverrideInSimSpace=true
    bOverrideAngularDamping=true
    bUseSphericalLimits=true
    gravityScale=0.32
    simulationSpace=RootRelative
    alphaBoolBlend {blendOption=Linear}
    boundBone {boneName=dyn_hair_bk_03}
    componentAppliedLinearAccClamp=<100 100 100>
    componentLinearVelScale=<50 50 1>
    physicsBodyDefinitions [{
      collisionType=CustomSphere
      sphereCollisionRadius=3
      boundBone {boneName=dyn_hair_bk_03}
      boxExtents=<12 40 40>
      constraintSetup {angularLimitsMax=<0 12 45>, angularLimitsMin=<0 -12 -75>, angularTarget=<0 0 0>}
      localJointOffset=<4 0 0>
     }]
    sphericalLimits [{limitRadius=38, drivingBone {boneName=spine_05}, sphereLocalOffset=<-12 26 0>}]
   }
   Anim Dynamics - Bone: dyn_hair_bk_04(ComponentPose=@n2, Alpha=@n1, ExternalForce="0,0,0") {
    angularDampingOverride=0.9999
    angularSpringConstant=1000
    bAngularSpring=true
    bGravityOverrideInSimSpace=true
    bOverrideAngularDamping=true
    bUseSphericalLimits=true
    gravityScale=0.32
    simulationSpace=RootRelative
    alphaBoolBlend {blendOption=Linear}
    boundBone {boneName=dyn_hair_bk_04}
    componentAppliedLinearAccClamp=<100 100 100>
    componentLinearVelScale=<50 50 1>
    physicsBodyDefinitions [{
      collisionType=CustomSphere
      sphereCollisionRadius=2.5
      boundBone {boneName=dyn_hair_bk_04}
      boxExtents=<10 30 30>
      constraintSetup {angularLimitsMax=<0 20 35>, angularLimitsMin=<0 -20 -75>, angularTarget=<0 0 0>}
      localJointOffset=<3 0 0>
     }]
    sphericalLimits [{limitRadius=38, drivingBone {boneName=spine_05}, sphereLocalOffset=<-12 26 0>}]
   }
   Anim Dynamics - Bone: dyn_hair_bk_r_01(Alpha=1, ExternalForce="0,0,0") {
    angularDampingOverride=1
    angularSpringConstant=1000
    bAngularSpring=true
    bGravityOverrideInSimSpace=true
    bOverrideAngularDamping=true
    bUsePlanarLimit=false
    gravityScale=0.33
    simulationSpace=BoneRelative
    alphaBoolBlend {blendOption=Linear}
    boundBone {boneName=dyn_hair_bk_r_01}
    componentAppliedLinearAccClamp=<100 100 100>
    componentLinearVelScale=<50 50 50>
    physicsBodyDefinitions [{
      collisionType=CustomSphere
      sphereCollisionRadius=3
      boundBone {boneName=dyn_hair_bk_r_01}
      boxExtents=<5 30 30>
      constraintSetup {angularLimitsMax=<15 20 30>, angularLimitsMin=<-15 -35 -65>, angularTarget=<0 0 0>}
      localJointOffset=<0 0 -6.5>
     }]
    relativeSpaceBone {boneName=head}
    sphericalLimits [
     {limitRadius=5, drivingBone {boneName=neck_02}, sphereLocalOffset=<0 0 0>}
     {limitRadius=45, drivingBone {boneName=spine_05}, sphereLocalOffset=<-30 20 8>}
    ]
   }
   n3 = Anim Dynamics - Bone: dyn_hair_bk_r_02(Alpha=1, ExternalForce="0,0,0") {
    angularDampingOverride=1
    angularSpringConstant=1000
    bAngularSpring=true
    bGravityOverrideInSimSpace=true
    bOverrideAngularDamping=true
    bUseSphericalLimits=true
    gravityScale=0.33
    alphaBoolBlend {blendOption=Linear}
    boundBone {boneName=dyn_hair_bk_r_02}
    componentAppliedLinearAccClamp=<100 100 100>
    componentLinearVelScale=<50 50 50>
    physicsBodyDefinitions [{
      collisionType=CustomSphere
      sphereCollisionRadius=2.2
      boundBone {boneName=dyn_hair_bk_r_02}
      boxExtents=<5 40 40>
      constraintSetup {angularLimitsMax=<0 20 30>, angularLimitsMin=<0 -20 -75>, angularTarget=<0 0 0>}
      localJointOffset=<4 0 0>
     }]
    sphericalLimits [
     {limitRadius=42, drivingBone {boneName=spine_05}, sphereLocalOffset=<-25 15 0>}
     {limitRadius=5, drivingBone {boneName=neck_02}, sphereLocalOffset=<0 0 0>}
    ]
   }
   Anim Dynamics - Bone: dyn_hair_bk_r_03(ComponentPose=@n3, Alpha=1, ExternalForce="0,0,0") {
    angularDampingOverride=1
    angularSpringConstant=1000
    bAngularSpring=true
    bGravityOverrideInSimSpace=true
    bOverrideAngularDamping=true
    bUseSphericalLimits=true
    gravityScale=0.33
    alphaBoolBlend {blendOption=Linear}
    boundBone {boneName=dyn_hair_bk_r_03}
    componentAppliedLinearAccClamp=<100 100 100>
    componentLinearVelScale=<50 50 50>
    physicsBodyDefinitions [{
      collisionType=CustomSphere
      sphereCollisionRadius=2
      boundBone {boneName=dyn_hair_bk_r_03}
      boxExtents=<5 40 40>
      constraintSetup {angularLimitsMax=<0 20 30>, angularLimitsMin=<0 -20 -75>, angularTarget=<0 0 0>}
      localJointOffset=<1 0 0>
     }]
    sphericalLimits [
     {limitRadius=41, drivingBone {boneName=spine_05}, sphereLocalOffset=<-25 15 0>}
     {limitRadius=5, drivingBone {boneName=neck_02}, sphereLocalOffset=<0 0 0>}
    ]
   }
   Anim Dynamics - Bone: dyn_hair_bk_l_01(Alpha=1, ExternalForce="0,0,0") {
    angularDampingOverride=0.99
    angularSpringConstant=1000
    bGravityOverrideInSimSpace=true
    bOverrideAngularDamping=true
    bUsePlanarLimit=false
    gravityScale=0.33
    simulationSpace=BoneRelative
    alphaBoolBlend {blendOption=Linear}
    boundBone {boneName=dyn_hair_bk_l_01}
    componentAppliedLinearAccClamp=<100 100 100>
    componentLinearVelScale=<50 50 50>
    physicsBodyDefinitions [{
      collisionType=CustomSphere
      sphereCollisionRadius=1
      boundBone {boneName=dyn_hair_bk_l_01}
      boxExtents=<10 20 20>
      constraintSetup {angularLimitsMax=<0 5 30>, angularLimitsMin=<0 -5 -65>, angularTarget=<0 0 0>}
      localJointOffset=<5 0 0>
     }]
    relativeSpaceBone {boneName=root}
    sphericalLimits [
     {limitRadius=5, drivingBone {boneName=neck_02}, sphereLocalOffset=<0 0 0>}
     {limitRadius=42, drivingBone {boneName=spine_05}, sphereLocalOffset=<-25 15 0>}
    ]
   }
   n4 = Anim Dynamics - Bone: dyn_hair_bk_l_02(Alpha=@n1, ExternalForce="0,0,0") {
    angularDampingOverride=1
    angularSpringConstant=1000
    bAngularSpring=true
    bGravityOverrideInSimSpace=true
    bOverrideAngularDamping=true
    bUseSphericalLimits=true
    gravityScale=0.6
    simulationSpace=RootRelative
    alphaBoolBlend {blendOption=Linear}
    boundBone {boneName=dyn_hair_bk_l_02}
    componentAppliedLinearAccClamp=<100 100 100>
    componentLinearVelScale=<50 50 50>
    physicsBodyDefinitions [{
      collisionType=CustomSphere
      sphereCollisionRadius=2.25
      boundBone {boneName=dyn_hair_bk_l_02}
      boxExtents=<5 40 40>
      constraintSetup {angularLimitsMax=<0 20 30>, angularLimitsMin=<0 -20 -75>, angularTarget=<0 0 0>}
      localJointOffset=<4 0 0>
     }]
    sphericalLimits [
     {limitRadius=42, drivingBone {boneName=spine_05}, sphereLocalOffset=<-25 15 0>}
     {limitRadius=5, drivingBone {boneName=neck_02}, sphereLocalOffset=<0 0 0>}
    ]
   }
   n5 = Anim Dynamics - Bone: dyn_hair_bk_l_03(ComponentPose=@n4, Alpha=@n1, ExternalForce="0,0,0") {
    angularDampingOverride=1
    angularSpringConstant=1000
    bAngularSpring=true
    bGravityOverrideInSimSpace=true
    bOverrideAngularDamping=true
    bUseSphericalLimits=true
    gravityScale=0.6
    simulationSpace=RootRelative
    alphaBoolBlend {blendOption=Linear}
    boundBone {boneName=dyn_hair_bk_l_03}
    componentAppliedLinearAccClamp=<100 100 100>
    componentLinearVelScale=<50 50 50>
    physicsBodyDefinitions [{
      collisionType=CustomSphere
      sphereCollisionRadius=2
      boundBone {boneName=dyn_hair_bk_l_03}
      boxExtents=<5 40 40>
      constraintSetup {angularLimitsMax=<0 20 30>, angularLimitsMin=<0 -20 -75>, angularTarget=<0 0 0>}
      localJointOffset=<3 0 0>
     }]
    sphericalLimits [
     {limitRadius=42, drivingBone {boneName=spine_05}, sphereLocalOffset=<-27 15 0>}
     {limitRadius=5, drivingBone {boneName=neck_02}, sphereLocalOffset=<0 0 0>}
    ]
   }
   Anim Dynamics - Bone: dyn_hair_bk_l_04(ComponentPose=@n5, Alpha=@n1, ExternalForce="0,0,0") {
    angularDampingOverride=1
    angularSpringConstant=1000
    bAngularSpring=true
    bGravityOverrideInSimSpace=true
    bOverrideAngularDamping=true
    bUseSphericalLimits=true
    gravityScale=0.6
    simulationSpace=RootRelative
    alphaBoolBlend {blendOption=Linear}
    boundBone {boneName=dyn_hair_bk_l_04}
    componentAppliedLinearAccClamp=<100 100 100>
    componentLinearVelScale=<50 50 50>
    physicsBodyDefinitions [{
      collisionType=CustomSphere
      sphereCollisionRadius=2.5
      boundBone {boneName=dyn_hair_bk_l_04}
      boxExtents=<5 40 40>
      constraintSetup {angularLimitsMax=<0 20 30>, angularLimitsMin=<0 -20 -75>, angularTarget=<0 0 0>}
      localJointOffset=<5 0 0>
     }]
    sphericalLimits [
     {limitRadius=41, drivingBone {boneName=spine_05}, sphereLocalOffset=<-27 15 0>}
     {limitRadius=5, drivingBone {boneName=neck_02}, sphereLocalOffset=<0 0 0>}
    ]
   }
   n6 = "Transform (Modify) Bone - Bone: dyn_hair_r_01"(ComponentPose=Local To Component(), Alpha=0.7, Translation="0,0,0", Rotation="-30.837,21.175,-9.310", Scale="1,1,1") {
    rotationMode=BMM_Additive
    rotationSpace=BCS_BoneSpace
    alphaBoolBlend {blendOption=Linear}
    alphaScaleBiasClamp {bInterpResult=true, interpSpeedDecreasing=15, interpSpeedIncreasing=15}
    boneToModify {boneName=dyn_hair_r_01}
    rotation=<rot -30.837 21.175 -9.31>
   }
   n7 = "Transform (Modify) Bone - Bone: dyn_hair_r_02"(ComponentPose=@n6, Alpha=0.7, Translation="0,0,0", Rotation="-11.056,-5.859,1.127", Scale="1,1,1") {
    rotationMode=BMM_Additive
    rotationSpace=BCS_BoneSpace
    alphaBoolBlend {blendOption=Linear}
    alphaScaleBiasClamp {bInterpResult=true, interpSpeedDecreasing=15, interpSpeedIncreasing=15}
    boneToModify {boneName=dyn_hair_r_02}
    rotation=<rot -11.056 -5.859 1.127>
   }
   n8 = "Transform (Modify) Bone - Bone: dyn_hair_bk_r_02"(ComponentPose=@n7, Alpha=0.7, Translation="0,0,0", Rotation="-3.148,10.885,-1.788", Scale="1,1,1") {
    rotationMode=BMM_Additive
    rotationSpace=BCS_BoneSpace
    alphaBoolBlend {blendOption=Linear}
    alphaScaleBiasClamp {bInterpResult=true, interpSpeedDecreasing=15, interpSpeedIncreasing=15}
    boneToModify {boneName=dyn_hair_bk_r_02}
    rotation=<rot -3.148 10.8851 -1.788>
   }
   n9 = "Transform (Modify) Bone - Bone: dyn_hair_bk_01"(ComponentPose=@n8, Alpha=0.7, Translation="0,0,0", Rotation="11.418,-30.059,1.272", Scale="1,1,1") {
    rotationMode=BMM_Additive
    rotationSpace=BCS_BoneSpace
    alphaBoolBlend {blendOption=Linear}
    alphaScaleBiasClamp {bInterpResult=true, interpSpeedDecreasing=15, interpSpeedIncreasing=15}
    boneToModify {boneName=dyn_hair_bk_01}
    rotation=<rot 11.418 -30.059 1.272>
   }
   n10 = "Transform (Modify) Bone - Bone: dyn_hair_bk_02"(ComponentPose=@n9, Alpha=0.7, Translation="0,0,0", Rotation="-3.148,10.885,-1.788", Scale="1,1,1") {
    rotationMode=BMM_Additive
    rotationSpace=BCS_BoneSpace
    alphaBoolBlend {blendOption=Linear}
    alphaScaleBiasClamp {bInterpResult=true, interpSpeedDecreasing=15, interpSpeedIncreasing=15}
    boneToModify {boneName=dyn_hair_bk_02}
    rotation=<rot -3.148 10.8851 -1.788>
   }
   Component To Local(ComponentPose=@n10)
   "Transform (Modify) Bone - Bone: dyn_hair_bk_r_01"(Alpha=0.7, Translation="0,0,0", Rotation="11.418,-30.059,1.272", Scale="1,1,1") {
    rotationMode=BMM_Additive
    rotationSpace=BCS_BoneSpace
    alphaBoolBlend {blendOption=Linear}
    alphaScaleBiasClamp {bInterpResult=true, interpSpeedDecreasing=15, interpSpeedIncreasing=15}
    boneToModify {boneName=dyn_hair_bk_r_01}
    rotation=<rot 11.418 -30.059 1.272>
   }
   n11 = "Transform (Modify) Bone - Bone: dyn_hair_bk_l_01"(Alpha=1, Translation="0,0,0", Rotation="1.918,-46.036,0.521", Scale="1,1,1") {
    rotationMode=BMM_Additive
    rotationSpace=BCS_BoneSpace
    alphaBoolBlend {blendOption=Linear}
    alphaScaleBiasClamp {bInterpResult=true, interpSpeedDecreasing=15, interpSpeedIncreasing=15}
    boneToModify {boneName=dyn_hair_bk_l_01}
    rotation=<rot 1.918 -46.0361 0.521>
   }
   n12 = "Transform (Modify) Bone - Bone: dyn_hair_bk_l_02"(ComponentPose=@n11, Alpha=1, Translation="0,0,0", Rotation="-3.616,-21.636,0.183", Scale="1,1,1") {
    rotationMode=BMM_Additive
    rotationSpace=BCS_BoneSpace
    alphaBoolBlend {blendOption=Linear}
    alphaScaleBiasClamp {bInterpResult=true, interpSpeedDecreasing=15, interpSpeedIncreasing=15}
    boneToModify {boneName=dyn_hair_bk_l_02}
    rotation=<rot -3.616 -21.636 0.183>
   }
   n13 = "Transform (Modify) Bone - Bone: dyn_hair_l_01"(ComponentPose=@n12, Alpha=1, Translation="0,0,0", Rotation="27.484,19.159,4.791", Scale="1,1,1") {
    rotationMode=BMM_Additive
    rotationSpace=BCS_BoneSpace
    alphaBoolBlend {blendOption=Linear}
    alphaScaleBiasClamp {bInterpResult=true, interpSpeedDecreasing=15, interpSpeedIncreasing=15}
    boneToModify {boneName=dyn_hair_l_01}
    rotation=<rot 27.484 19.159 4.791>
   }
   "Transform (Modify) Bone - Bone: dyn_hair_l_02"(ComponentPose=@n13, Alpha=1, Translation="0,0,0", Rotation="5.177,4.598,0.601", Scale="1,1,1") {
    rotationMode=BMM_Additive
    rotationSpace=BCS_BoneSpace
    alphaBoolBlend {blendOption=Linear}
    alphaScaleBiasClamp {bInterpResult=true, interpSpeedDecreasing=15, interpSpeedIncreasing=15}
    boneToModify {boneName=dyn_hair_l_02}
    rotation=<rot 5.177 4.598 0.601>
   }
   n14 = "Transform (Modify) Bone - Bone: dyn_strap_parent"(
    ComponentPose=Local To Component(LocalPose=Use cached pose 'AnimCache'())
    Alpha=Get strap_rot_alpha() -> (strap_rot_alpha:double)
    Translation="0,0,0"
    Rotation="0.000,0.000,25"
    Scale="1,1,1"
   ) {
    rotationMode=BMM_Additive
    rotationSpace=BCS_BoneSpace
    alphaBoolBlend {blendOption=Linear}
    alphaScaleBiasClamp {bInterpResult=true, interpSpeedDecreasing=20, interpSpeedIncreasing=20}
    boneToModify {boneName=dyn_strap_parent}
    rotation {roll=25}
   }
   n15 = "Transform (Modify) Bone - Bone: dyn_strap_parent"(ComponentPose=@n14, Alpha=Get head_rot_bck() -> (head_rot_bck:double), Translation="-0.000,-1,3.25", Rotation="0.000,0.000,8", Scale="1,1,1") {
    rotationMode=BMM_Additive
    rotationSpace=BCS_BoneSpace
    translationMode=BMM_Additive
    translationSpace=BCS_BoneSpace
    alphaBoolBlend {blendOption=Linear}
    alphaScaleBiasClamp {bInterpResult=true, interpSpeedDecreasing=20, interpSpeedIncreasing=20}
    boneToModify {boneName=dyn_strap_parent}
    rotation {roll=8}
    translation {y=-1, z=3.25}
   }
   n16 = "Transform (Modify) Bone - Bone: dyn_strap_parent"(ComponentPose=@n15, Alpha=1, Translation="0,-1,0", Rotation="4.925,-16.465,-20", Scale="1,1,1") {
    rotationMode=BMM_Additive
    rotationSpace=BCS_BoneSpace
    translationMode=BMM_Additive
    translationSpace=BCS_BoneSpace
    alphaBoolBlend {blendOption=Linear}
    boneToModify {boneName=dyn_strap_parent}
    rotation=<rot 4.925 -16.465 -20>
    translation {y=-1}
   }
   n17 = Look At - Bone: dyn_strapB_01(ComponentPose=@n16, Alpha=1) {
    alphaBoolBlend {blendOption=Linear}
    boneToModify {boneName=dyn_strapB_01}
    lookAtLocation=<-5 -17 -5>
    lookAtTarget {boneReference {boneName=spine_05}}
    lookAt_Axis {axis=<1 0>}
    lookUp_Axis {axis=<0 1>}
   }
   n18 = Look At - Bone: dyn_strapA_01(ComponentPose=@n17, Alpha=1) {
    alphaBoolBlend {blendOption=Linear}
    boneToModify {boneName=dyn_strapA_01}
    lookAtLocation=<-5 -17 2>
    lookAtTarget {boneReference {boneName=spine_05}}
    lookAt_Axis {axis=<1 0>}
    lookUp_Axis {axis=<0 1>}
   }
   n19 = Trail controller - Bone: dyn_strapA_05(ComponentPose=@n18, Alpha=1) {
    bUsePlanarLimit=true
    chainLength=3
    debugLifeTime=0.005
    maxDeltaTime=0.0333
    relaxationSpeedScale=3
    planarLimits [
     {drivingBone {boneName=spine_05}, planeTransform {location=<0 -10 0>, rotation=<rot 0 0 -90>, scale=<1 1 1>}}
     {drivingBone {boneName=spine_05}, planeTransform {location=<10 -10 0>, rotation=<rot 0 24.9999 -90>, scale=<1 1 1>}}
    ]
    rotationLimits [{limitMax=<180 180 180>}, {limitMax=<180 180 180>}, {limitMax=<180 180 180>}]
    rotationOffsets [<0 0 0>, <0 0 0>, <0 0 0>]
    trailBone {boneName=dyn_strapA_05}
   }
   n20 = Trail controller - Bone: dyn_strapB_05(ComponentPose=@n19, Alpha=1) {
    bUsePlanarLimit=true
    chainLength=3
    maxDeltaTime=0.0333
    relaxationSpeedScale=3
    planarLimits [
     {drivingBone {boneName=spine_05}, planeTransform {location=<0 -10 0>, rotation=<rot 0 0 -90>, scale=<1 1 1>}}
     {drivingBone {boneName=spine_05}, planeTransform {location=<10 -10 0>, rotation=<rot 0 24.9999 -90>, scale=<1 1 1>}}
    ]
    rotationLimits [{limitMax=<180 180 180>}, {limitMax=<180 180 180>}, {limitMax=<180 180 180>}]
    rotationOffsets [<0 0 0>, <0 0 0>, <0 0 0>]
    trailBone {boneName=dyn_strapB_05}
   }
   n21 = "Transform (Modify) Bone - Bone: dyn_strapA_04"(ComponentPose=@n20, Alpha=1, Translation="0,0,0", Rotation="-20,0,0", Scale="1,1,1") {
    rotationMode=BMM_Additive
    rotationSpace=BCS_BoneSpace
    alphaBoolBlend {blendOption=Linear}
    boneToModify {boneName=dyn_strapA_04}
    rotation {pitch=-20}
   }
   n22 = "Transform (Modify) Bone - Bone: dyn_strapB_04"(ComponentPose=@n21, Alpha=1, Translation="0,0,0", Rotation="-20,0,0", Scale="1,1,1") {
    rotationMode=BMM_Additive
    rotationSpace=BCS_BoneSpace
    alphaBoolBlend {blendOption=Linear}
    boneToModify {boneName=dyn_strapB_04}
    rotation {pitch=-20}
   }
   n23 = "Transform (Modify) Bone - Bone: dyn_strapA_02"(ComponentPose=@n22, Alpha=1, Translation="0.000,0.000,0.803", Rotation="0.383,0.000,0.000", Scale="1,1,1") {
    rotationMode=BMM_Additive
    rotationSpace=BCS_BoneSpace
    translationMode=BMM_Additive
    translationSpace=BCS_BoneSpace
    alphaBoolBlend {blendOption=Linear}
    boneToModify {boneName=dyn_strapA_02}
    rotation {pitch=0.383}
    translation {z=0.803}
   }
   n24 = "Transform (Modify) Bone - Bone: dyn_strapB_02"(ComponentPose=@n23, Alpha=1, Translation="0.000,0.000,0.776", Rotation="3.779,0.000,0.000", Scale="1,1,1") {
    rotationMode=BMM_Additive
    rotationSpace=BCS_BoneSpace
    translationMode=BMM_Additive
    translationSpace=BCS_BoneSpace
    alphaBoolBlend {blendOption=Linear}
    boneToModify {boneName=dyn_strapB_02}
    rotation {pitch=3.779}
    translation {z=0.776}
   }
   n25 = Flap Node - Bone: dyn_hair_bk_01!(ComponentPose:Component Space Pose Link Structure=@n24, Alpha:float=1) -> (Pose:Component Space Pose Link Structure) {
    is=title FortAnimGraphNode_Flap
    bUseClamp=true
    targetClampMax=100
    targetClampMin=-100
    targetComponent=RotationZ
    alphaBoolBlend {blendOption=Linear}
    alphaScaleBiasClamp {bInterpResult=true, interpSpeedDecreasing=15, interpSpeedIncreasing=15}
    sourceDrives [{
      bUseRange=false
      drivingCurve=null
      multiplier=1
      rangeMax=1
      rangeMin=-1
      remappedMax=1
      remappedMin=0
      sourceComponent=RotationZ
      useQuaternion=true
      sourceBone {boneName=head}
     }]
    targetBone {boneName=dyn_hair_bk_01}
   }
   n26 = Flap Node - Bone: dyn_hair_bk_01(ComponentPose=@n25, Alpha=1) {
    bUseClamp=true
    targetClampMin=-3.14
    targetComponent=TranslationY
    alphaBoolBlend {blendOption=Linear}
    alphaScaleBiasClamp {bInterpResult=true, interpSpeedDecreasing=15, interpSpeedIncreasing=15}
    sourceDrives [{
      bUseRange=false
      drivingCurve=null
      multiplier=-0.35
      rangeMax=1
      rangeMin=-1
      remappedMax=1
      remappedMin=0
      sourceComponent=RotationZ
      useQuaternion=true
      sourceBone {boneName=head}
     }]
    targetBone {boneName=dyn_hair_bk_01}
   }
   n27 = Flap Node - Bone: dyn_hair_bk_01(ComponentPose=@n26, Alpha=1) {
    bUseClamp=true
    targetClampMax=10
    targetComponent=TranslationX
    alphaBoolBlend {blendOption=Linear}
    alphaScaleBiasClamp {bInterpResult=true, interpSpeedDecreasing=15, interpSpeedIncreasing=15}
    sourceDrives [{
      bUseRange=false
      drivingCurve=null
      multiplier=10
      rangeMax=1
      rangeMin=-1
      remappedMax=1
      remappedMin=0
      sourceComponent=RotationZ
      useQuaternion=true
      sourceBone {boneName=head}
     }]
    targetBone {boneName=dyn_hair_bk_01}
   }
   n28 = Flap Node - Bone: dyn_hair_bk_l_01!(ComponentPose:Component Space Pose Link Structure=@n27, Alpha:float=1) -> (Pose:Component Space Pose Link Structure) {
    is=title FortAnimGraphNode_Flap
    bUseClamp=true
    targetClampMax=100
    targetClampMin=-100
    targetComponent=RotationZ
    alphaBoolBlend {blendOption=Linear}
    alphaScaleBiasClamp {bInterpResult=true, interpSpeedDecreasing=15, interpSpeedIncreasing=15}
    sourceDrives [{
      bUseRange=false
      drivingCurve=null
      multiplier=1
      rangeMax=1
      rangeMin=-1
      remappedMax=1
      remappedMin=0
      sourceComponent=RotationZ
      useQuaternion=true
      sourceBone {boneName=head}
     }]
    targetBone {boneName=dyn_hair_bk_l_01}
   }
   n29 = Flap Node - Bone: dyn_hair_bk_l_01(ComponentPose=@n28, Alpha=1) {
    bUseClamp=true
    targetClampMin=-3.14
    targetComponent=TranslationY
    alphaBoolBlend {blendOption=Linear}
    alphaScaleBiasClamp {bInterpResult=true, interpSpeedDecreasing=15, interpSpeedIncreasing=15}
    sourceDrives [{
      bUseRange=false
      drivingCurve=null
      multiplier=-0.35
      rangeMax=1
      rangeMin=-1
      remappedMax=1
      remappedMin=0
      sourceComponent=RotationZ
      useQuaternion=true
      sourceBone {boneName=head}
     }]
    targetBone {boneName=dyn_hair_bk_l_01}
   }
   n30 = Flap Node - Bone: dyn_hair_bk_l_01(ComponentPose=@n29, Alpha=1) {
    bUseClamp=true
    targetClampMax=10
    targetComponent=TranslationX
    alphaBoolBlend {blendOption=Linear}
    alphaScaleBiasClamp {bInterpResult=true, interpSpeedDecreasing=15, interpSpeedIncreasing=15}
    sourceDrives [{
      bUseRange=false
      drivingCurve=null
      multiplier=5
      rangeMax=1
      rangeMin=-1
      remappedMax=1
      remappedMin=0
      sourceComponent=RotationZ
      useQuaternion=true
      sourceBone {boneName=head}
     }]
    targetBone {boneName=dyn_hair_bk_l_01}
   }
   n31 = Flap Node - Bone: dyn_hair_bk_r_01!(ComponentPose:Component Space Pose Link Structure=@n30, Alpha:float=1) -> (Pose:Component Space Pose Link Structure) {
    is=title FortAnimGraphNode_Flap
    bUseClamp=true
    targetClampMax=100
    targetClampMin=-100
    targetComponent=RotationZ
    alphaBoolBlend {blendOption=Linear}
    alphaScaleBiasClamp {bInterpResult=true, interpSpeedDecreasing=15, interpSpeedIncreasing=15}
    sourceDrives [{
      bUseRange=false
      drivingCurve=null
      multiplier=1
      rangeMax=1
      rangeMin=-1
      remappedMax=1
      remappedMin=0
      sourceComponent=RotationZ
      useQuaternion=true
      sourceBone {boneName=head}
     }]
    targetBone {boneName=dyn_hair_bk_r_01}
   }
   n32 = Flap Node - Bone: dyn_hair_bk_r_01(ComponentPose=@n31, Alpha=1) {
    bUseClamp=true
    targetClampMin=-3.14
    targetComponent=TranslationY
    alphaBoolBlend {blendOption=Linear}
    alphaScaleBiasClamp {bInterpResult=true, interpSpeedDecreasing=15, interpSpeedIncreasing=15}
    sourceDrives [{
      bUseRange=false
      drivingCurve=null
      multiplier=-0.35
      rangeMax=1
      rangeMin=-1
      remappedMax=1
      remappedMin=0
      sourceComponent=RotationZ
      useQuaternion=true
      sourceBone {boneName=head}
     }]
    targetBone {boneName=dyn_hair_bk_r_01}
   }
   n33 = Flap Node - Bone: dyn_hair_bk_r_01(ComponentPose=@n32, Alpha=1) {
    bUseClamp=true
    targetClampMax=10
    targetComponent=TranslationX
    alphaBoolBlend {blendOption=Linear}
    alphaScaleBiasClamp {bInterpResult=true, interpSpeedDecreasing=15, interpSpeedIncreasing=15}
    sourceDrives [{
      bUseRange=false
      drivingCurve=null
      multiplier=10
      rangeMax=1
      rangeMin=-1
      remappedMax=1
      remappedMin=0
      sourceComponent=RotationZ
      useQuaternion=true
      sourceBone {boneName=head}
     }]
    targetBone {boneName=dyn_hair_bk_r_01}
   }
   n34 = Flap Node - Bone: dyn_hair_bk_r_01(ComponentPose=@n33, Alpha=1) {
    bUseClamp=true
    targetClampMax=10
    targetComponent=RotationY
    alphaBoolBlend {blendOption=Linear}
    alphaScaleBiasClamp {bInterpResult=true, interpSpeedDecreasing=15, interpSpeedIncreasing=15}
    sourceDrives [{
      bUseRange=false
      drivingCurve=null
      multiplier=2.25
      rangeMax=1
      rangeMin=-1
      remappedMax=1
      remappedMin=0
      sourceComponent=RotationZ
      useQuaternion=true
      sourceBone {boneName=head}
     }]
    targetBone {boneName=dyn_hair_bk_r_01}
   }
   n35 = Flap Node - Bone: dyn_hair_r_01!(ComponentPose:Component Space Pose Link Structure=@n34, Alpha:float=1) -> (Pose:Component Space Pose Link Structure) {
    is=title FortAnimGraphNode_Flap
    bUseClamp=true
    targetClampMax=100
    targetClampMin=-0.5
    targetComponent=RotationZ
    alphaBoolBlend {blendOption=Linear}
    sourceDrives [{
      bUseRange=false
      drivingCurve=null
      multiplier=1
      rangeMax=1
      rangeMin=-1
      remappedMax=1
      remappedMin=0
      sourceComponent=RotationZ
      useQuaternion=true
      sourceBone {boneName=head}
     }]
    targetBone {boneName=dyn_hair_r_01}
   }
   n36 = Flap Node - Bone: dyn_hair_r_01(ComponentPose=@n35, Alpha=1) {
    bUseClamp=true
    targetClampMax=10
    targetComponent=RotationY
    alphaBoolBlend {blendOption=Linear}
    alphaScaleBiasClamp {bInterpResult=true, interpSpeedDecreasing=15, interpSpeedIncreasing=15}
    sourceDrives [{
      bUseRange=false
      drivingCurve=null
      multiplier=0.5
      rangeMax=1
      rangeMin=-1
      remappedMax=1
      remappedMin=0
      sourceComponent=RotationZ
      useQuaternion=true
      sourceBone {boneName=head}
     }]
    targetBone {boneName=dyn_hair_r_01}
   }
   n37 = "Transform (Modify) Bone - Bone: dyn_strap_parent"(
    ComponentPose=Local To Component(LocalPose=Use cached pose 'AnimCache'())
    Alpha=Get strap_rot_alpha()
    Translation="0,0,0"
    Rotation="0.000,0.000,25"
    Scale="1,1,1"
   ) {
    rotationMode=BMM_Additive
    rotationSpace=BCS_BoneSpace
    alphaBoolBlend {blendOption=Linear}
    alphaScaleBiasClamp {bInterpResult=true, interpSpeedDecreasing=20, interpSpeedIncreasing=20}
    boneToModify {boneName=dyn_strap_parent}
    rotation {roll=25}
   }
   n38 = "Transform (Modify) Bone - Bone: dyn_strap_parent"(ComponentPose=@n37, Alpha=Get head_rot_bck(), Translation="-0.000,-1,3.25", Rotation="0.000,0.000,8", Scale="1,1,1") {
    rotationMode=BMM_Additive
    rotationSpace=BCS_BoneSpace
    translationMode=BMM_Additive
    translationSpace=BCS_BoneSpace
    alphaBoolBlend {blendOption=Linear}
    alphaScaleBiasClamp {bInterpResult=true, interpSpeedDecreasing=20, interpSpeedIncreasing=20}
    boneToModify {boneName=dyn_strap_parent}
    rotation {roll=8}
    translation {y=-1, z=3.25}
   }
   n39 = "Transform (Modify) Bone - Bone: dyn_strap_parent"(ComponentPose=@n38, Alpha=1, Translation="-0.000,-1,0.000", Rotation="0,0,0", Scale="1,1,1") {translationMode=BMM_Additive, alphaBoolBlend {blendOption=Linear}, boneToModify {boneName=dyn_strap_parent}, translation {y=-1}}
   n40 = "Transform (Modify) Bone - Bone: dyn_strap_parent"(ComponentPose=@n39, Alpha=Get straps_stand_rot() -> (straps_stand_rot:double), Translation="0,0,0", Rotation="0,0,-25", Scale="1,1,1") {
    rotationMode=BMM_Additive
    rotationSpace=BCS_BoneSpace
    alphaBoolBlend {blendOption=Linear}
    boneToModify {boneName=dyn_strap_parent}
    rotation {roll=-25}
   }
   n41 = Get straps_vector() -> (straps_vector:Vector)
   n42 = RigidBody!(
    ComponentPose=@n40
    Alpha=Get straps_standing_alpha() -> (straps_standing_alpha:double)
    ExternalForce=Get straps_force_vector() -> (straps_force_vector:Vector)
    ComponentLinearAccScale:Vector=@n41
    ComponentLinearVelScale:Vector=@n41
   ) -> (Pose) {
    bClampLinearTranslationLimitToRefPose=true
    bOverrideWorldGravity=true
    overridePhysicsAsset=&/F_MED_Toxic_Kitty_Straps_Physics
    simulationSpace=BaseBoneSpace
    alphaBoolBlend {blendOption=Linear}
    alphaScaleBiasClamp {bInterpResult=true}
    baseBoneRef {boneName=root}
    componentAppliedLinearAccClamp=<3000 1500 2000>
    componentLinearAccScale {z=0.5}
    componentLinearVelScale {z=0.5}
    overrideWorldGravity {z=-980}
   }
   Output Pose(Result=Blend Poses by bool(
     BlendPose_0=Component To Local(ComponentPose=@n36)
     BlendPose_1=Component To Local(ComponentPose=@n42)
     bActiveValue=Get rigids_off() -> (rigids_off:bool)
    ) {blendType=Linear}) {layerGroup=None}
   n43 = Flap Node - Bone: dyn_hair_bk_01(ComponentPose=Local To Component(LocalPose=Copy Pose From Mesh() {bUseAttachedParent=true}), Alpha=1) {
    bUseClamp=true
    targetClampMax=100
    targetComponent=RotationZ
    alphaBoolBlend {blendOption=Linear}
    alphaScaleBiasClamp {bInterpResult=true, interpSpeedDecreasing=15, interpSpeedIncreasing=15}
    sourceDrives [{
      bUseRange=false
      drivingCurve=null
      multiplier=0.3
      rangeMax=1
      rangeMin=-1
      remappedMax=1
      remappedMin=0
      sourceComponent=RotationZ
      useQuaternion=true
      sourceBone {boneName=head}
     }]
    targetBone {boneName=dyn_hair_bk_01}
   }
   n44 = Flap Node - Bone: dyn_hair_bk_01(ComponentPose=@n43, Alpha=1) {
    bUseClamp=true
    targetClampMin=-3.14
    targetComponent=TranslationY
    alphaBoolBlend {blendOption=Linear}
    alphaScaleBiasClamp {bInterpResult=true, interpSpeedDecreasing=15, interpSpeedIncreasing=15}
    sourceDrives [{
      bUseRange=false
      drivingCurve=null
      multiplier=-0.35
      rangeMax=1
      rangeMin=-1
      remappedMax=1
      remappedMin=0
      sourceComponent=RotationZ
      useQuaternion=true
      sourceBone {boneName=head}
     }]
    targetBone {boneName=dyn_hair_bk_01}
   }
   n45 = Flap Node - Bone: dyn_hair_bk_01(ComponentPose=@n44, Alpha=1) {
    bUseClamp=true
    targetClampMax=10
    targetComponent=TranslationX
    alphaBoolBlend {blendOption=Linear}
    alphaScaleBiasClamp {bInterpResult=true, interpSpeedDecreasing=15, interpSpeedIncreasing=15}
    sourceDrives [{
      bUseRange=false
      drivingCurve=null
      multiplier=10
      rangeMax=1
      rangeMin=-1
      remappedMax=1
      remappedMin=0
      sourceComponent=RotationZ
      useQuaternion=true
      sourceBone {boneName=head}
     }]
    targetBone {boneName=dyn_hair_bk_01}
   }
   n46 = Flap Node - Bone: dyn_hair_bk_l_01(ComponentPose=@n45, Alpha=1) {
    bUseClamp=true
    targetClampMax=100
    targetComponent=RotationZ
    alphaBoolBlend {blendOption=Linear}
    alphaScaleBiasClamp {bInterpResult=true, interpSpeedDecreasing=15, interpSpeedIncreasing=15}
    sourceDrives [{
      bUseRange=false
      drivingCurve=null
      multiplier=0.3
      rangeMax=1
      rangeMin=-1
      remappedMax=1
      remappedMin=0
      sourceComponent=RotationZ
      useQuaternion=true
      sourceBone {boneName=head}
     }]
    targetBone {boneName=dyn_hair_bk_l_01}
   }
   n47 = Flap Node - Bone: dyn_hair_bk_l_01(ComponentPose=@n46, Alpha=1) {
    bUseClamp=true
    targetClampMin=-3.14
    targetComponent=TranslationY
    alphaBoolBlend {blendOption=Linear}
    alphaScaleBiasClamp {bInterpResult=true, interpSpeedDecreasing=15, interpSpeedIncreasing=15}
    sourceDrives [{
      bUseRange=false
      drivingCurve=null
      multiplier=-0.35
      rangeMax=1
      rangeMin=-1
      remappedMax=1
      remappedMin=0
      sourceComponent=RotationZ
      useQuaternion=true
      sourceBone {boneName=head}
     }]
    targetBone {boneName=dyn_hair_bk_l_01}
   }
   n48 = Flap Node - Bone: dyn_hair_bk_l_01(ComponentPose=@n47, Alpha=1) {
    bUseClamp=true
    targetClampMax=10
    targetComponent=TranslationX
    alphaBoolBlend {blendOption=Linear}
    alphaScaleBiasClamp {bInterpResult=true, interpSpeedDecreasing=15, interpSpeedIncreasing=15}
    sourceDrives [{
      bUseRange=false
      drivingCurve=null
      multiplier=5
      rangeMax=1
      rangeMin=-1
      remappedMax=1
      remappedMin=0
      sourceComponent=RotationZ
      useQuaternion=true
      sourceBone {boneName=head}
     }]
    targetBone {boneName=dyn_hair_bk_l_01}
   }
   n49 = Flap Node - Bone: dyn_hair_bk_r_01(ComponentPose=@n48, Alpha=1) {
    bUseClamp=true
    targetClampMax=100
    targetComponent=RotationZ
    alphaBoolBlend {blendOption=Linear}
    alphaScaleBiasClamp {bInterpResult=true, interpSpeedDecreasing=15, interpSpeedIncreasing=15}
    sourceDrives [{
      bUseRange=false
      drivingCurve=null
      multiplier=1
      rangeMax=1
      rangeMin=-1
      remappedMax=1
      remappedMin=0
      sourceComponent=RotationZ
      useQuaternion=true
      sourceBone {boneName=head}
     }]
    targetBone {boneName=dyn_hair_bk_r_01}
   }
   n50 = Flap Node - Bone: dyn_hair_bk_r_01(ComponentPose=@n49, Alpha=1) {
    bUseClamp=true
    targetClampMin=-3.14
    targetComponent=TranslationY
    alphaBoolBlend {blendOption=Linear}
    alphaScaleBiasClamp {bInterpResult=true, interpSpeedDecreasing=15, interpSpeedIncreasing=15}
    sourceDrives [{
      bUseRange=false
      drivingCurve=null
      multiplier=-0.35
      rangeMax=1
      rangeMin=-1
      remappedMax=1
      remappedMin=0
      sourceComponent=RotationZ
      useQuaternion=true
      sourceBone {boneName=head}
     }]
    targetBone {boneName=dyn_hair_bk_r_01}
   }
   n51 = Flap Node - Bone: dyn_hair_bk_r_01(ComponentPose=@n50, Alpha=1) {
    bUseClamp=true
    targetClampMax=10
    targetComponent=TranslationX
    alphaBoolBlend {blendOption=Linear}
    alphaScaleBiasClamp {bInterpResult=true, interpSpeedDecreasing=15, interpSpeedIncreasing=15}
    sourceDrives [{
      bUseRange=false
      drivingCurve=null
      multiplier=10
      rangeMax=1
      rangeMin=-1
      remappedMax=1
      remappedMin=0
      sourceComponent=RotationZ
      useQuaternion=true
      sourceBone {boneName=head}
     }]
    targetBone {boneName=dyn_hair_bk_r_01}
   }
   n52 = Flap Node - Bone: dyn_hair_bk_r_01(ComponentPose=@n51, Alpha=1) {
    bUseClamp=true
    targetClampMax=10
    targetComponent=RotationY
    alphaBoolBlend {blendOption=Linear}
    alphaScaleBiasClamp {bInterpResult=true, interpSpeedDecreasing=15, interpSpeedIncreasing=15}
    sourceDrives [{
      bUseRange=false
      drivingCurve=null
      multiplier=2.25
      rangeMax=1
      rangeMin=-1
      remappedMax=1
      remappedMin=0
      sourceComponent=RotationZ
      useQuaternion=true
      sourceBone {boneName=head}
     }]
    targetBone {boneName=dyn_hair_bk_r_01}
   }
   n53 = Flap Node - Bone: dyn_hair_r_01(ComponentPose=@n52, Alpha=1) {
    bUseClamp=true
    targetClampMax=100
    targetClampMin=-0.5
    targetComponent=RotationZ
    alphaBoolBlend {blendOption=Linear}
    sourceDrives [{
      bUseRange=false
      drivingCurve=null
      multiplier=1
      rangeMax=1
      rangeMin=-1
      remappedMax=1
      remappedMin=0
      sourceComponent=RotationZ
      useQuaternion=true
      sourceBone {boneName=head}
     }]
    targetBone {boneName=dyn_hair_r_01}
   }
   n54 = Flap Node - Bone: dyn_hair_r_01(ComponentPose=@n53, Alpha=1) {
    bUseClamp=true
    targetClampMax=10
    targetComponent=RotationY
    alphaBoolBlend {blendOption=Linear}
    alphaScaleBiasClamp {bInterpResult=true, interpSpeedDecreasing=15, interpSpeedIncreasing=15}
    sourceDrives [{
      bUseRange=false
      drivingCurve=null
      multiplier=0.5
      rangeMax=1
      rangeMin=-1
      remappedMax=1
      remappedMin=0
      sourceComponent=RotationZ
      useQuaternion=true
      sourceBone {boneName=head}
     }]
    targetBone {boneName=dyn_hair_r_01}
   }
   n55 = Anim Dynamics - Bone: dyn_hair_01_l(ComponentPose=@n54, Alpha=1, ExternalForce="0,0,0") {
    angularDampingOverride=1
    angularSpringConstant=1500
    bAngularSpring=true
    bGravityOverrideInSimSpace=true
    bOverrideAngularDamping=true
    gravityScale=0.29
    simulationSpace=RootRelative
    alphaBoolBlend {blendOption=Linear}
    boundBone {boneName=dyn_hair_01_l}
    physicsBodyDefinitions [{
      sphereCollisionRadius=0
      boundBone {boneName=dyn_hair_01_l}
      boxExtents=<13 20 20>
      constraintSetup {angularLimitsMax=<0 10 20>, angularLimitsMin=<0 -35 -40>, angularTarget=<0 0 0>}
      localJointOffset=<10 0 0>
     }]
   }
   n56 = Anim Dynamics - Bone: dyn_hair_02_l(ComponentPose=@n55, Alpha=1, ExternalForce="0,0,0") {
    angularDampingOverride=1
    angularSpringConstant=1500
    bAngularSpring=true
    bGravityOverrideInSimSpace=true
    bOverrideAngularDamping=true
    gravityScale=0.29
    simulationSpace=RootRelative
    alphaBoolBlend {blendOption=Linear}
    boundBone {boneName=dyn_hair_02_l}
    physicsBodyDefinitions [{
      sphereCollisionRadius=0
      boundBone {boneName=dyn_hair_02_l}
      boxExtents=<13 20 20>
      constraintSetup {angularLimitsMax=<0 10 6>, angularLimitsMin=<0 -25 -40>, angularTarget=<0 0 0>}
      localJointOffset=<10 0 0>
     }]
   }
   n57 = Anim Dynamics - Bone: dyn_hair_03_l(ComponentPose=@n56, Alpha=1, ExternalForce="0,0,0") {
    angularDampingOverride=1
    angularSpringConstant=1500
    bAngularSpring=true
    bGravityOverrideInSimSpace=true
    bOverrideAngularDamping=true
    gravityScale=0.29
    simulationSpace=RootRelative
    alphaBoolBlend {blendOption=Linear}
    boundBone {boneName=dyn_hair_03_l}
    physicsBodyDefinitions [{
      sphereCollisionRadius=0
      boundBone {boneName=dyn_hair_03_l}
      boxExtents=<13 20 20>
      constraintSetup {angularLimitsMax=<0 15 5>, angularLimitsMin=<0 -25 -40>, angularTarget=<0 0 0>}
      localJointOffset=<10 0 0>
     }]
   }
   n58 = RigidBody!(ComponentPose=@n57, Alpha, ExternalForce) -> (Pose) {
    bClampLinearTranslationLimitToRefPose=true
    bOverrideWorldGravity=true
    overridePhysicsAsset=&/F_MED_Toxic_Kitty_FaceAcc_Physics
    simulationSpace=BaseBoneSpace
    alphaBoolBlend {blendOption=Linear}
    alphaScaleBiasClamp {bInterpResult=true}
    baseBoneRef {boneName=neck_02}
    componentAppliedLinearAccClamp=<800 1000 1000>
    componentLinearAccScale=<0.35 0.4 0.4>
    componentLinearVelScale=<0.35 0.4 0.4>
    overrideWorldGravity {z=-950}
   }
   Save cached pose 'PreCrouchCache'(Pose=Component To Local(ComponentPose=@n58)) {cacheName=PreCrouchCache}
   Save cached pose 'AnimCache'(Pose=Blend Poses by bool(
     BlendPose_0=Use cached pose 'PreCrouchCache'()
     BlendPose_1=Use cached pose 'PreCrouchCache'()
     bActiveValue=Get crouch_moving() -> (crouch_moving:bool)
    ) {blendType=Linear}) {cacheName=AnimCache}
   Reroute Node(InputPin:Wildcard) -> (OutputPin:Wildcard)
  ]}
 EventGraph {nodes [
   Event BlueprintUpdateAnimation()
   > Is Valid(InputObject=Get player_pawn() -> (player_pawn:Fort Player Pawn Athena Object Reference))
   > Is Valid(InputObject=Get Skeletal_Mesh() -> (Skeletal_Mesh:Skeletal Mesh Component Object Reference))
   > n1 = Sequence!(execute) -> (then_0, then_1, then_2:Exec, then_3:Exec, then_4:Exec, then_5:Exec)
   n2 = Get CurrentLOD()
   n3 = Get bRigidBodyEnabled()
   Set rigids_off(execute=@n1.then_4, rigids_off:bool=OR Boolean(A=NOT Boolean(A=@n3), B="integer > integer"(A=@n2, B=1) {is=title K2Node_CallFunction})) -> (Output_Get:bool)
   n4 = Get PawnSpeed2D()
   n5 = Branch(
    execute=@n1.then_0
    Condition="float <= float"!(A:double=MapRangeClamped(Value=@n4, InRangeA=0, InRangeB=150), B:double=0) -> (ReturnValue:bool) {is=title K2Node_CallFunction}
   )
   > Set straps_vector(straps_vector:Vector="10,800,800") -> (Output_Get:Vector)
   > Set hair_vector(hair_vector:Vector="1,1,1") -> (Output_Get:Vector)
   > Set standing_still(standing_still:bool=true) -> (Output_Get:bool)
   > Set strap_rot_alpha(strap_rot_alpha:double=0) -> (Output_Get:double)
   > Set straps_standing_alpha(straps_standing_alpha:double=0.9) -> (Output_Get:double)
   > Set straps_vector(straps_vector="0,0,1")
   Set straps_vector(execute=@n5.else, straps_vector="1,0.05,0.05")
   > Set hair_vector(hair_vector="0.5,0.5,0.5")
   > Set standing_still(standing_still=false)
   > Set straps_standing_alpha(straps_standing_alpha=0.9)
   > Set straps_stand_rot(straps_stand_rot:double=0) -> (Output_Get:double)
   > Set strap_rot_alpha(strap_rot_alpha=0)
   > n6 = Branch(Condition=Get isMovingBackward() -> (isMovingBackward:bool))
   n7 = Get player_pawn()
   n8 = Get Actor Rotation!(self:Actor Object Reference=@n7) -> (ReturnValue:Rotator) {is=title K2Node_CallFunction}
   n9 = BreakVector(InVec=vector * vector(
     A:Vector=Get Rotation X Vector!(InRot:Rotator=@n8) -> (ReturnValue:Vector) {is=title K2Node_CallFunction}
     B:Vector=GetVelocity!(self:Actor Object Reference=@n7) -> (ReturnValue:Vector)
    ) -> (ReturnValue:Vector) {is=title K2Node_CommutativeAssociativeBinaryOperator})
   n10 = float + float(A=@n9, B=@n9.Y) {is=title K2Node_CommutativeAssociativeBinaryOperator}
   n11 = Branch(execute=@n6.else, Condition="float < float"!(A:double=@n10, B:double=415) -> (ReturnValue:bool) {is=title K2Node_CallFunction})
   > Set straps_vector(straps_vector="0.25,2.5,0.6")
   > n12 = Branch(Condition=Get bIsCrouching())
   > Set strap_rot_alpha(strap_rot_alpha=0)
   Set strap_rot_alpha(execute=@n12.else, strap_rot_alpha=0.2)
   n13 = Branch(execute=@n11.else, Condition=Get bIsJumping())
   > Set straps_vector(straps_vector="0, 0,0.65")
   > Set strap_rot_alpha(strap_rot_alpha=0.15)
   Set straps_vector(execute=@n13.else, straps_vector="0,1,0.65")
   > Set strap_rot_alpha(strap_rot_alpha=0.2)
   n14 = GetPartSkeletalMeshComponent()
   n15 = Get Distance Between Two Sockets!(
    Component:Skeletal Mesh Component Object Reference=@n14
    SocketOrBoneNameA:Name=dyn_collar_r
    SocketSpaceA:ERelativeTransformSpace Enum=RTS_Component
    SocketOrBoneNameB:Name=dyn_hair_r_03
    SocketSpaceB:ERelativeTransformSpace Enum=RTS_Component
    bRemapRange:bool=true
    InRangeMin:float=-10
    InRangeMax:float=10
    OutRangeMin:float=0
    OutRangeMax:float=1
   ) -> (ReturnValue:float) {is=title K2Node_CallFunction}
   Set hair_dist(execute=@n1.then_2, hair_dist:double=float - float(A=@n15, B=1) {is=title K2Node_CallFunction}) -> (Output_Get:double)
   > Set rot_hair_r(rot_hair_r:Rotator=MakeRotator(Pitch=float * float(A=@n15, B=-5))) -> (Output_Get:Rotator)
   n16 = float - float(
    A=Get Distance Between Two Sockets(
     Component=@n14
     SocketOrBoneNameA=dyn_collar_l
     SocketSpaceA=RTS_Component
     SocketOrBoneNameB=dyn_hair_l_03
     SocketSpaceB=RTS_Component
     bRemapRange=true
     InRangeMin=0
     InRangeMax=100
     OutRangeMin=0
     OutRangeMax=1
    )
    B=1
   )
   > Set rot_hair_l(rot_hair_l:Rotator=MakeRotator(Yaw=float * float(A=@n16, B=5))) -> (Output_Get:Rotator)
   > n17 = Branch(Condition=Get bIsTargeting())
   > Set target_hair_r(target_hair_r:double=1) -> (Output_Get:double)
   Set target_hair_r(execute=@n17.else, target_hair_r=0)
   Set pawn_speed(execute=@n1.then_1, pawn_speed:double=@n10) -> (Output_Get:double)
   > Set isMovingBackward(isMovingBackward:bool="float < float"(A=@n10, B=0)) -> (Output_Get:bool)
   n18 = GetOwnerSkeletalMeshComponent()
   n19 = BreakRotator(InRot=BreakTransform(InTransform=GetDeltaTransformFromRefPose(self=@n18, BoneName=upperarm_r, BaseName=spine_05)).Rotation)
   n20 = MapRangeClamped(Value=@n19.Yaw, InRangeA=100, InRangeB=129)
   Set arm_up_r(execute=@n1.then_3, arm_up_r:double=@n20) -> (Output_Get:double)
   > n21 = Branch(Condition=Get bIsCrouching())
   > Set crouch_alpha(crouch_alpha:double=@n20) -> (Output_Get:double)
   > n22 = Branch(Condition="float >= float"(A=Get PawnSpeed2D(), B=100) {is=title K2Node_CallFunction})
   > Set crouch_moving(crouch_moving:bool=true) -> (Output_Get:bool)
   Set crouch_moving(execute=@n22.else, crouch_moving=false)
   Set crouch_alpha(execute=@n21.else, crouch_alpha=0)
   > Set crouch_moving(crouch_moving=false)
   n23 = BreakTransform(InTransform=GetDeltaTransformFromRefPose(self=GetOwnerSkeletalMeshComponent(), BoneName=head, BaseName=spine_05))
   Set head_rot_bck(execute=@n1.then_5, head_rot_bck:double=MapRangeClamped(Value=BreakRotator(InRot=@n23.Rotation).Yaw, InRangeA=0, InRangeB=-60)) -> (Output_Get:double)
   Event BlueprintInitializeAnimation()
   n24 = TryGetPawnOwner()
   > Is Valid(InputObject=@n24)
   > n25 = Cast To FortPlayerPawnAthena(Object=@n24)
   > Set player_pawn(player_pawn:Fort Player Pawn Athena Object Reference=@n25) -> (Output_Get:Fort Player Pawn Athena Object Reference)
   n26 = GetPartSkeletalMeshComponent()
   > Is Valid(InputObject=@n26)
   > Set Skeletal_Mesh(Skeletal_Mesh:Skeletal Mesh Component Object Reference=@n26) -> (Output_Get:Skeletal Mesh Component Object Reference)
   TryGetPawnOwner()
   TryGetPawnOwner()
   MapRangeClamped(Value=@n19.Yaw, InRangeA=35, InRangeB=-20)
   MapRangeClamped(Value=@n19.Yaw, InRangeA=85, InRangeB=125)
   GetOwnerSkeletalMeshComponent()
  ]}
]
variables [
 arm_up_r {type=number}
 crouch_alpha {type=number}
 crouch_moving {type=boolean}
 hair_alpha {type=number, default=0.875}
 hair_dist {type=number}
 hair_vector {type=Vector}
 head_rot_bck {type=number}
 isMovingBackward {type=boolean}
 pawn_speed {type=number}
 player_pawn {type=FortPlayerPawnAthena}
 rigids_off {type=boolean}
 rot_hair_l {type=Rotator}
 rot_hair_r {type=Rotator}
 Skeletal_Mesh {type=SkeletalMeshComponent}
 standing_still {type=boolean}
 strap_rot_alpha {type=number}
 straps_force_vector {type=Vector}
 straps_stand_rot {type=number}
 straps_standing_alpha {type=number}
 straps_vector {type=Vector}
 target_hair_r {type=number}
]
